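# Control Rig for the face accessory part of the male medium GateHound Spectral body, correcting lower hair bones as the head moves.
# Two spherical pose readers measure head orientation relative to spine_05 in differently oriented regions.
# Each weights a Modify Transforms step on hair_lower bones, the first through a cubic remap curve with a strong scale change.
# Both readers have debug drawing enabled.
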ControlRigBlueprint /BRCosmetics/Characters/Player/Male/Medium/Bodies/M_MED_GateHound_Spectral/Meshes/Parts/M_MED_GateHound_Spectral_FaceAcc_CtrlRig
graphs [
 RigVMFunctionLibrary {links [], nodes []}
 RigVMModel {nodes [
   RigUnit_BeginExecution = &RigUnit_BeginExecution()
   n1 = &RigUnit_SphericalPoseReader(
    ActiveRegionScaleFactors=(PositiveWidth=1,NegativeWidth=3,PositiveHeight=3,NegativeHeight=3)
    ActiveRegionSize=0.1
    Debug=(bDrawDebug=true,bDraw2D=False,bDrawLocalAxes=False,DebugScale=25,DebugSegments=20,DebugThickness=0.25)
    DriverAxis=(X=1,Y=0,Z=0)
    DriverItem=(Type=Bone,Name="head")
    FalloffRegionScaleFactors=(PositiveWidth=1,NegativeWidth=1,PositiveHeight=1,NegativeHeight=1)
    FalloffSize=0.4
    FlipHeightScaling=False
    FlipWidthScaling=False
    OptionalParentItem=(Type=Bone,Name="spine_05")
    RotationOffset=(X=-90,Y=0,Z=5)
    ExecutePin=@RigUnit_BeginExecution.ExecutePin
   )
   RigVMFunction_AnimEvalRichCurve = &RigVMFunction_AnimEvalRichCurve(
    Curve=(EditorCurveData=(Keys=((InterpMode=RCIM_Cubic),(InterpMode=RCIM_Cubic,Time=0.7221,Value=0.448,ArriveTangent=1,LeaveTangent=1),(Time=1,Value=1))))
    SourceMinimum=0
    TargetMinimum=0
    Value=@n1.OutputParam
   )
   > &RigUnit_ModifyTransforms(
    ItemToModify="((Item=(Type=Bone,Name=\"hair_lower_root\"),Transform=(Rotation=(X=0,Y=0,Z=0.6088,W=0.7934),Translation=(X=-0.2697,Y=2.6866,Z=0),Scale3D=(X=0.1,Y=0.1,Z=0.1))),(Item=(Type=Bone,Name=\"hair_lower_left_root\"),Transform=(Rotation=(X=0,Y=0,Z=0,W=1),Translation=(X= ...(truncated)"
    WeightMinimum=0
    Weight=@RigVMFunction_AnimEvalRichCurve.Result
   )
   > n2 = &RigUnit_SphericalPoseReader(
    ActiveRegionScaleFactors=(PositiveWidth=1,NegativeWidth=2.3,PositiveHeight=2,NegativeHeight=1)
    ActiveRegionSize=0.1
    Debug=(bDrawDebug=true,bDraw2D=False,bDrawLocalAxes=False,DebugScale=25,DebugSegments=20,DebugThickness=0.25)
    DriverAxis=(X=1,Y=0,Z=0)
    DriverItem=(Type=Bone,Name="head")
    FalloffRegionScaleFactors=(PositiveWidth=1,NegativeWidth=1,PositiveHeight=1,NegativeHeight=1)
    FalloffSize=0.32
    FlipHeightScaling=False
    FlipWidthScaling=False
    OptionalParentItem=(Type=Bone,Name="spine_05")
    RotationOffset=(X=-40,Y=0,Z=50)
   )
   > &RigUnit_ModifyTransforms(
    ItemToModify="((Item=(Type=Bone,Name=\"hair_lower_root\"),Transform=(Rotation=(X=0,Y=0,Z=0.001,W=1),Translation=(X=0,Y=0,Z=0),Scale3D=(X=1,Y=1,Z=1))),(Item=(Type=Bone,Name=\"hair_lower_left_root\"),Transform=(Rotation=(X=0,Y=0,Z=0,W=1),Translation=(X=0,Y=0,Z=0),Scale3D=(X=1 ...(truncated)"
    WeightMinimum=0
    Weight=@n2.OutputParam
   )
  ]}
]
hierarchy [
 root {loc=<0 0 0>, rot=<0 0 0 1>}
 pelvis {parent=root, loc=<0 -0.7071 93.1191>, rot=<0.0127 0.707 -0.0127 -0.707>}
 spine_01 {parent=pelvis, loc=<6.7494 -0.6388 0>, rot=<0 0 0.034 -0.9994>}
 spine_02 {parent=spine_01, loc=<7.2518 -0.035 0>, rot=<0 0 0.0238 -0.9997>}
 spine_03 {parent=spine_02, loc=<6.0104 0 0>, rot=<0 0 0.0149 -0.9999>}
 spine_04 {parent=spine_03, loc=<8.1024 0 0>, rot=<0 0 -0.0132 -0.9999>}
 spine_05 {parent=spine_04, loc=<10.0146 0 0>, rot=<0 0 -0.0015 -1>}
 neck_01 {parent=spine_05, loc=<11.3138 0.1529 0>, rot=<0 0 -0.2003 -0.9797>}
 neck_02 {parent=neck_01, loc=<4.5627 0 0>, rot=<0 0 0.0504 -0.9987>}
 head {parent=neck_02, loc=<5.3914 0 0>, rot=<0 0 0.115 -0.9934>}
 hair_root {parent=head, loc=<0 0 0>, rot=<0 0 0 1>}
 dyn_hair_04 {parent=hair_root, loc=<11.9014 -2.3208 2.2674>, rot=<0.6293 -0.2051 0.1763 -0.7285>}
 dyn_hair_05 {parent=hair_root, loc=<12.0856 2.5341 2.1377>, rot=<0.5704 -0.3202 0.2553 -0.712>}
 dyn_hair_06 {parent=hair_root, loc=<11.5929 2.9868 -1.8707>, rot=<0.6806 -0.2775 0.3155 -0.6002>}
 dyn_hair_07 {parent=hair_root, loc=<12.6031 1.1579 0.3513>, rot=<0.5813 -0.3547 0.31 -0.6635>}
 dyn_hair_08 {parent=hair_root, loc=<12.028 -3.08 -1.4679>, rot=<0.7759 -0.1659 0.2308 -0.5632>}
 dyn_hair_09 {parent=hair_root, loc=<11.404 -4.8346 -3.6851>, rot=<0.4532 -0.1767 0.0894 -0.8691>}
 dyn_hair_10 {parent=hair_root, loc=<10.2358 -5.7685 -4.9821>, rot=<0.7911 -0.1617 0.2402 -0.5388>}
 dyn_hair_11 {parent=hair_root, loc=<11.3045 -6.2681 1.1722>, rot=<0.7387 -0.2177 0.283 -0.5718>}
 dyn_hair_12 {parent=hair_root, loc=<10.6738 -6.4663 4.8327>, rot=<0.4534 -0.2301 0.1196 -0.8528>}
 dyn_hair_28 {parent=hair_root, loc=<3.4464 -3.8509 -8.0691>, rot=<0.557 -0.3133 0.237 -0.7317>}
 dyn_hair_29 {parent=dyn_hair_28, loc=<4.1776 0 0>, rot=<-0.4544 0.2288 -0.2452 -0.8253>}
 dyn_hair_30 {parent=hair_root, loc=<7.4712 -2.3812 -5.5412>, rot=<0.3958 -0.4253 0.211 -0.7861>}
 dyn_hair_31 {parent=dyn_hair_30, loc=<3.539 0 0>, rot=<-0.2659 0.0969 -0.2966 -0.9121>}
 dyn_hair_32 {parent=hair_root, loc=<7.0079 -1.5372 6.4469>, rot=<-0.4233 -0.6839 0.5574 0.2059>}
 dyn_hair_33 {parent=dyn_hair_32, loc=<-3.5391 0 -0.0002>, rot=<-0.2659 0.0969 -0.2966 -0.9121>}
 dyn_hair_34 {parent=hair_root, loc=<3.4467 -3.8509 8.0691>, rot=<-0.3133 -0.557 0.7317 0.237>}
 dyn_hair_35 {parent=dyn_hair_34, loc=<-4.1774 0 0.0001>, rot=<-0.4544 0.2288 -0.2452 -0.8253>}
 hair_lower_root {parent=head, loc=<0 0 0>, rot=<0 0 0 1>}
 dyn_hair_15 {parent=hair_lower_root, loc=<-2.9301 -9.2719 0.4126>, rot=<0.5756 -0.4229 0.4711 -0.5176>}
 dyn_hair_16 {parent=hair_lower_root, loc=<1.55 -10.2281 0.1854>, rot=<0.649 -0.3397 0.4033 -0.5485>}
 hair_lower_right_root {parent=hair_lower_root, loc=<0 0 0>, rot=<0 0 0 1>}
 dyn_hair_01 {parent=hair_lower_right_root, loc=<-3.9983 -6.3912 5.3448>, rot=<0.6616 -0.4545 0.4829 -0.35>}
 dyn_hair_13 {parent=hair_lower_right_root, loc=<2.3411 -9.3286 4.6877>, rot=<0.7891 -0.2079 0.3896 -0.427>}
 dyn_hair_22 {parent=hair_lower_right_root, loc=<-2.2757 -4.3151 6.9572>, rot=<0.8786 -0.0743 0.4455 -0.1552>}
 dyn_hair_23 {parent=dyn_hair_22, loc=<0.6575 0.2094 -2.4845>, rot=<-0.0915 -0.0717 -0.1314 -0.9845>}
 dyn_hair_24 {parent=dyn_hair_22, loc=<3.277 3.8968 -0.5346>, rot=<0 -0.0251 0 -0.9997>}
 dyn_hair_36 {parent=hair_lower_right_root, loc=<-1.8271 -7.3212 3.5773>, rot=<0.6381 -0.4662 0.5467 -0.2766>}
 hair_lower_left_root {parent=hair_lower_root, loc=<0 0 0>, rot=<0 0 0 1>}
 dyn_hair_02 {parent=hair_lower_left_root, loc=<-3.748 -8.437 -6.5208>, rot=<0.4543 -0.515 0.3729 -0.624>}
 dyn_hair_03 {parent=hair_lower_left_root, loc=<-2.5326 -8.0793 -3.6639>, rot=<0.3735 -0.5366 0.2821 -0.7022>}
 dyn_hair_14 {parent=hair_lower_left_root, loc=<1.99 -9.0625 -4.2736>, rot=<0.3519 -0.4337 0.2336 -0.7959>}
 dyn_hair_25 {parent=hair_lower_left_root, loc=<-2.857 -3.0432 -5.7996>, rot=<0.4317 -0.5067 0.3225 -0.673>}
 dyn_hair_26 {parent=dyn_hair_25, loc=<2.7782 1.8459 -1.9344>, rot=<0.2593 0.0785 0.5626 -0.7811>}
 dyn_hair_27 {parent=dyn_hair_25, loc=<6.9182 0 0>, rot=<-0.1781 0.1537 -0.3326 -0.9132>}
 hair_mid_root {parent=head, loc=<5.056 -6.4723 0>, rot=<0 0 0 1>}
 dyn_hair_17 {parent=hair_mid_root, loc=<2.8308 -3.6575 -3.7378>, rot=<0.2583 -0.3149 0.0856 -0.9093>}
 dyn_hair_18 {parent=hair_mid_root, loc=<2.1479 -3.3172 5.9112>, rot=<0.9234 -0.0603 0.3348 -0.1779>}
 dyn_hair_19 {parent=hair_mid_root, loc=<2.3037 -4.285 1.0615>, rot=<0.7519 -0.0871 0.3399 -0.5582>}
 dyn_hair_20 {parent=hair_mid_root, loc=<4.9857 -2.1707 -1.2238>, rot=<0.4839 -0.2566 0.1484 -0.8234>}
 dyn_hair_21 {parent=hair_mid_root, loc=<4.9611 -2.5476 3.4986>, rot=<0.9004 -0.0905 0.3487 -0.2439>}
]
